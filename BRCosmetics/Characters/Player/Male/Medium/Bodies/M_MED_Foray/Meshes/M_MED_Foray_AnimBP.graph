AnimBlueprint /BRCosmetics/Characters/Player/Male/Medium/Bodies/M_MED_Foray/Meshes/M_MED_Foray_AnimBP
target_skeleton=&/M_MED_Foray_Skeleton
graphs [
 AnimGraph {nodes [
   "Transform (Modify) Bone - Bone: upperarm_fr_volume_r"(Alpha=Get pistol_ads_up_alpha_r() -> (pistol_ads_up_alpha_r:double), Translation="5,-7,0", Rotation="0,0,0", Scale="1,1,1") {
    rotationMode=BMM_Additive
    rotationSpace=BCS_BoneSpace
    translationMode=BMM_Additive
    translationSpace=BCS_BoneSpace
    alphaBoolBlend {blendOption=Linear}
    boneToModify {boneName=upperarm_fr_volume_r}
    translation=<5 -7>
   }
   n1 = Break Pose Deformation Query Results Arms!(PoseDeformationQueryResults_Arms:"Pose Deformation Query Results Arms Structure (by ref)"=Break Pose Deformation Query Arms!(PoseDeformationQuery_Arms:"Pose Deformation Query Arms Structure (by ref)"=Get psd_arms() -> (psd_arms:Pose Deformation Query Arms Structure)) -> (
     QueryResults:Pose Deformation Query Results Arms Structure
     ArmUpL:Pose Deformation Query Structure
     ArmUpR:Pose Deformation Query Structure
     ArmForwardL:Pose Deformation Query Structure
     ArmForwardR:Pose Deformation Query Structure
     ArmBackL:Pose Deformation Query Structure
     ArmBackR:Pose Deformation Query Structure
     ArmForwardUpL:Pose Deformation Query Structure
     ArmForwardUpR:Pose Deformation Query Structure
     ArmForwardDownL:Pose Deformation Query Structure
     ArmForwardDownR:Pose Deformation Query Structure
    ) {is=title K2Node_BreakStruct}) -> (
    ArmUpL:float=~
    ArmUpR:float=~
    ArmForwardL:float=~
    ArmForwardR:float=~
    ArmBackL:float=~
    ArmBackR:float=~
    ArmForwardUpL:float=~
    ArmForwardUpR:float=~
    ArmForwardDownL:float=~
    ArmForwardDownR:float=~
   ) {is=title K2Node_BreakStruct}
   n2 = Control Rig!(
    Source=Copy Pose From Mesh() {bUseAttachedParent=true}
    Alpha
    thigh_rot_l:double=Get thigh_rot_l()
    elbow_rot_l:double=Get elbow_rot_l() -> (elbow_rot_l:double)
    elbow_rot_r:double=Get elbow_rot_r() -> (elbow_rot_r:double)
    thigh_rot_r:double=Get thigh_rot_r()
    knee_rot_l:double=Get knee_rot_l() -> (knee_rot_l:double)
    knee_rot_r:double=Get knee_rot_r() -> (knee_rot_r:double)
    psd_arm_up_alpha_l:double=@n1
    psd_arm_up_alpha_r:double=@n1.ArmUpR
    psd_arm_fwd_alpha_l:double=@n1.ArmForwardL
    psd_arm_fwd_alpha_r:double=@n1.ArmForwardR
    psd_arm_bk_alpha_l:double=@n1.ArmBackL
    psd_arm_bk_alpha_r:double=@n1.ArmBackR
    psd_arm_fwd_up_alpha_l:double=@n1.ArmForwardUpL
    psd_arm_fwd_up_alpha_r:double=@n1.ArmForwardUpR
    psd_arm_fwd_dn_alpha_l:double=@n1.ArmForwardDownL
    psd_arm_fwd_dn_alpha_r:double=@n1.ArmForwardDownR
   ) -> (Pose) {
    alphaBoolBlend {blendOption=Linear}
    controlRigAssetReference {blueprintRigClass=&/M_MED_Muscle_Volume_ControlRig.M_MED_Muscle_Volume_ControlRig_C}
   }
   n3 = "Constraint - dyn_shirts_arm_l (P)"(ComponentPose=Local To Component(LocalPose=@n2), Alpha=1, ConstraintWeights_0=1) {
    alphaBoolBlend {blendOption=Linear}
    boneToModify {boneName=dyn_shirts_arm_l}
    constraintSetup [{targetBone {boneName=upperarm_twist_02_l}}]
    constraintWeights [1]
   }
   n4 = "Constraint - dyn_shirts_arm_r (P)"(ComponentPose=@n3, Alpha=1, ConstraintWeights_0=1) {
    alphaBoolBlend {blendOption=Linear}
    boneToModify {boneName=dyn_shirts_arm_r}
    constraintSetup [{targetBone {boneName=upperarm_twist_02_r}}]
    constraintWeights [1]
   }
   n5 = Anim Dynamics - Bone: dyn_shirts_arm_l(ComponentPose=@n4, Alpha=0.7, ExternalForce="0,0,0") {
    angularDampingOverride=1
    angularSpringConstant=20000
    bAngularSpring=true
    bGravityOverrideInSimSpace=true
    bOverrideAngularDamping=true
    bUseGravityOverride=true
    gravityScale=0.35
    simulationSpace=RootRelative
    alphaBoolBlend {blendOption=Linear}
    boundBone {boneName=dyn_shirts_arm_l}
    componentLinearAccScale=<0.5 0.5 0.5>
    componentLinearVelScale=<0.5 0.5 0.5>
    gravityOverride {z=-980}
    physicsBodyDefinitions [{
      sphereCollisionRadius=0
      boundBone {boneName=dyn_shirts_arm_l}
      boxExtents=<20 30 30>
      constraintSetup {angularLimitsMax=<0 0 5>, angularLimitsMin=<0 -5 -5>, angularTarget=<1 0 0>}
      localJointOffset=<4 0 0>
     }]
   }
   n6 = Anim Dynamics - Bone: dyn_shirts_arm_r(ComponentPose=@n5, Alpha=0.7, ExternalForce="0,0,0") {
    angularDampingOverride=1
    angularSpringConstant=20000
    bAngularSpring=true
    bGravityOverrideInSimSpace=true
    bOverrideAngularDamping=true
    bUseGravityOverride=true
    gravityScale=0.35
    simulationSpace=RootRelative
    alphaBoolBlend {blendOption=Linear}
    boundBone {boneName=dyn_shirts_arm_r}
    componentLinearAccScale=<0.5 0.5 0.5>
    componentLinearVelScale=<0.5 0.5 0.5>
    gravityOverride {z=-980}
    physicsBodyDefinitions [{
      sphereCollisionRadius=0
      boundBone {boneName=dyn_shirts_arm_r}
      boxExtents=<20 30 30>
      constraintSetup {angularLimitsMax=<0 0 5>, angularLimitsMin=<0 -5 -5>, angularTarget=<1 0 0>}
      localJointOffset=<-4 0 0>
     }]
   }
   n7 = Get pistol_ads_up_alpha_r()
   n8 = "Transform (Modify) Bone - Bone: upperarm_fr_volume_l"(ComponentPose=@n6, Alpha=@n7, Translation="-3,3,0", Rotation="0,0,0", Scale="1,1,1") {
    rotationMode=BMM_Additive
    rotationSpace=BCS_BoneSpace
    scaleMode=BMM_Additive
    scaleSpace=BCS_BoneSpace
    translationMode=BMM_Additive
    translationSpace=BCS_BoneSpace
    alphaBoolBlend {blendOption=Linear}
    boneToModify {boneName=upperarm_fr_volume_l}
    translation=<-3 3>
   }
   n9 = "Transform (Modify) Bone - Bone: upperarm_fr_volume_r"(ComponentPose=@n8, Alpha=@n7, Translation="7,-7,0", Rotation="0,0,0", Scale="1,1,1") {
    rotationMode=BMM_Additive
    rotationSpace=BCS_BoneSpace
    scaleMode=BMM_Additive
    scaleSpace=BCS_BoneSpace
    translationMode=BMM_Additive
    translationSpace=BCS_BoneSpace
    alphaBoolBlend {blendOption=Linear}
    boneToModify {boneName=upperarm_fr_volume_r}
    translation=<7 -7>
   }
   n10 = "Transform (Modify) Bone - Bone: dyn_shirts_arm_r"(ComponentPose=@n9, Alpha=@n7, Translation="0,0,0", Rotation="0,0,0", Scale="1,0.85,0.85") {
    rotationMode=BMM_Additive
    rotationSpace=BCS_BoneSpace
    scaleMode=BMM_Additive
    scaleSpace=BCS_BoneSpace
    translationMode=BMM_Additive
    translationSpace=BCS_BoneSpace
    alphaBoolBlend {blendOption=Linear}
    boneToModify {boneName=dyn_shirts_arm_r}
    scale {y=0.85, z=0.85}
   }
   Output Pose(Result=Component To Local(ComponentPose=@n10)) {layerGroup=None}
   Save cached pose 'musclePose'() {cacheName=musclePose}
  ]}
 EventGraph {nodes [
   n1 = GetPartSkeletalMeshComponent()
   n2 = FN_PSDPoseReaderFNC!(
    execute
    startJnt:String=upperarm_l
    endJnt:String=lowerarm_l
    poseJnt:String=pose_arm_fwd_l
    angle:double=50
    debug:bool=false
    skeletalMeshComponent:Skeletal Mesh Component Object Reference=@n1
   ) -> (then, pose_output:double)
   > Set psd_arm_fwd_alpha_l(psd_arm_fwd_alpha_l:double=@n2) -> (Output_Get:double)
   n3 = GetPartSkeletalMeshComponent()
   > n4 = FN_PSDPoseReaderFNC(startJnt=upperarm_r, endJnt=lowerarm_r, poseJnt=pose_arm_fwd_r, angle=50, debug=false, skeletalMeshComponent=@n3)
   > Set psd_arm_fwd_alpha_r(psd_arm_fwd_alpha_r:double=@n4) -> (Output_Get:double)
   n5 = FN_PSDPoseReaderFNC(startJnt=upperarm_l, endJnt=lowerarm_l, poseJnt=pose_arm_bk_l, angle=50, debug=false, skeletalMeshComponent=@n1)
   > Set psd_arm_bk_alpha_l(psd_arm_bk_alpha_l:double=@n5) -> (Output_Get:double)
   > n6 = FN_PSDPoseReaderFNC(startJnt=upperarm_r, endJnt=lowerarm_r, poseJnt=pose_arm_bk_r, angle=50, debug=false, skeletalMeshComponent=@n3)
   > Set psd_arm_bk_alpha_r(psd_arm_bk_alpha_r:double=@n6) -> (Output_Get:double)
   n7 = FN_PSDPoseReaderFNC(startJnt=upperarm_l, endJnt=lowerarm_l, poseJnt=pose_arm_fwd_up_l, angle=45, debug=false, skeletalMeshComponent=@n1)
   > Set psd_arm_fwd_up_alpha_l(psd_arm_fwd_up_alpha_l:double=@n7) -> (Output_Get:double)
   > n8 = FN_PSDPoseReaderFNC(startJnt=upperarm_r, endJnt=lowerarm_r, poseJnt=pose_arm_fwd_up_r, angle=45, debug=false, skeletalMeshComponent=@n3)
   > Set psd_arm_fwd_up_alpha_r(psd_arm_fwd_up_alpha_r:double=@n8) -> (Output_Get:double)
   n9 = FN_PSDPoseReaderFNC(startJnt=upperarm_l, endJnt=lowerarm_l, poseJnt=pose_arm_fwd_dn_l, angle=40, debug=false, skeletalMeshComponent=@n1)
   > Set psd_arm_fwd_dn_alpha_l(psd_arm_fwd_dn_alpha_l:double=@n9) -> (Output_Get:double)
   > n10 = FN_PSDPoseReaderFNC(startJnt=upperarm_r, endJnt=lowerarm_r, poseJnt=pose_arm_fwd_dn_r, angle=40, debug=false, skeletalMeshComponent=@n3)
   > Set psd_arm_fwd_dn_alpha_r(psd_arm_fwd_dn_alpha_r:double=@n10) -> (Output_Get:double)
   n11 = FN_PSDPoseReaderFNC(startJnt=upperarm_l, endJnt=lowerarm_l, poseJnt=pose_arm_up_l, angle=70, debug=false, skeletalMeshComponent=@n1)
   > Set psd_arm_up_alpha_l(psd_arm_up_alpha_l:double=@n11) -> (Output_Get:double)
   > n12 = FN_PSDPoseReaderFNC(startJnt=upperarm_r, endJnt=lowerarm_r, poseJnt=pose_arm_up_r, angle=70, debug=false, skeletalMeshComponent=@n3)
   > Set psd_arm_up_alpha_r(psd_arm_up_alpha_r:double=@n12) -> (Output_Get:double)
   Event BlueprintUpdateAnimation()
   > n13 = Sequence!(execute) -> (then_0, then_1, then_2:Exec, then_3:Exec, then_4:Exec, then_5:Exec, then_6:Exec)
   n14 = Sequence!(execute=@n13.then_3) -> (then_0, then_1, then_2, then_3, then_4, then_5, then_6, then_7:Exec)
   > QueryPoseDeformation_Arms!(
    execute
    MeshComponent:Skeletal Mesh Component Object Reference=@n1
    ArmsDeformationQuery:"Pose Deformation Query Arms Structure (by ref)"=Get psd_arms()
   ) -> (then)
   n15 = Cast To FortPlayerAnimInstance(execute=@n14.then_6, Object=GetAnimInstance(self=GetOwnerSkeletalMeshComponent()))
   > n16 = Set isPistol(isPistol="Equal (Enum)"(A:EFortWeaponCoreAnimation Enum=Get WeaponCoreAnim(self=@n15), B:EFortWeaponCoreAnimation Enum=Pistol))
   > n17 = Branch(Condition=@n16)
   n18 = Get psd_arms()
   > Set pistol_ads_up_alpha_r(pistol_ads_up_alpha_r:double=Break Pose Deformation Query Results Arms(PoseDeformationQueryResults_Arms=Break Pose Deformation Query Arms(PoseDeformationQuery_Arms=@n18)) -> (
     ArmUpL=~
     ArmUpR=~
     ArmForwardL=~
     ArmForwardR=~
     ArmBackL=~
     ArmBackR=~
     ArmForwardUpL=~
     ArmForwardUpR=~
     ArmForwardDownL=~
     ArmForwardDownR=~
    ).ArmForwardUpR) -> (Output_Get:double)
   Set pistol_ads_up_alpha_r(execute=@n17.else, pistol_ads_up_alpha_r=0)
   n19 = GetDeltaTransformFromRefPose(self=GetOwnerSkeletalMeshComponent(), BoneName=lowerarm_l, BaseName=upperarm_l)
   Set elbow_rot_l(execute=@n13.then_0, elbow_rot_l:double=BreakRotator(InRot=BreakTransform(InTransform=@n19).Rotation).Yaw) -> (Output_Get:double)
   n20 = GetDeltaTransformFromRefPose(self=GetOwnerSkeletalMeshComponent(), BoneName=lowerarm_r, BaseName=upperarm_r)
   > Set elbow_rot_r(elbow_rot_r:double=BreakRotator(InRot=BreakTransform(InTransform=@n20).Rotation).Yaw) -> (Output_Get:double)
   n21 = GetDeltaTransformFromRefPose(self=GetOwnerSkeletalMeshComponent(), BoneName=calf_l, BaseName=thigh_l)
   Set knee_rot_l(execute=@n13.then_2, knee_rot_l:double=BreakRotator(InRot=BreakTransform(InTransform=@n21).Rotation).Yaw) -> (Output_Get:double)
   n22 = GetDeltaTransformFromRefPose(self=GetOwnerSkeletalMeshComponent(), BoneName=calf_r, BaseName=thigh_r)
   > Set knee_rot_r(knee_rot_r:double=BreakRotator(InRot=BreakTransform(InTransform=@n22).Rotation).Yaw) -> (Output_Get:double)
   n23 = GetDeltaTransformFromRefPose(self=GetOwnerSkeletalMeshComponent(), BoneName=thigh_l, BaseName=pelvis)
   Set thigh_rot_l(execute=@n13.then_1, thigh_rot_l=BreakRotator(InRot=BreakTransform(InTransform=@n23).Rotation).Yaw)
   n24 = GetDeltaTransformFromRefPose(self=GetOwnerSkeletalMeshComponent(), BoneName=thigh_r, BaseName=pelvis)
   > Set thigh_rot_r(thigh_rot_r=BreakRotator(InRot=BreakTransform(InTransform=@n24).Rotation).Yaw)
   TryGetPawnOwner()
   TryGetPawnOwner()
   Get psd_arm_fwd_up_alpha_r() -> (psd_arm_fwd_up_alpha_r:double)
  ]}
]
variables [
 elbow_rot_l {type=number}
 elbow_rot_r {type=number}
 isPistol {type=boolean}
 knee_rot_l {type=number}
 knee_rot_r {type=number}
 pistol_ads_up_alpha_r {type=number}
 psd_arm_bk_alpha_l {type=number}
 psd_arm_bk_alpha_r {type=number}
 psd_arm_fwd_alpha_l {type=number}
 psd_arm_fwd_alpha_r {type=number}
 psd_arm_fwd_dn_alpha_l {type=number}
 psd_arm_fwd_dn_alpha_r {type=number}
 psd_arm_fwd_up_alpha_l {type=number}
 psd_arm_fwd_up_alpha_r {type=number}
 psd_arm_up_alpha_l {type=number}
 psd_arm_up_alpha_r {type=number}
 psd_arms {
  type=PoseDeformationQuery_Arms
  default {
   armBackL {angle=50, bDrawDebug=false, endJoint=lowerarm_l, poseJoint=pose_arm_bk_l, startJoint=upperarm_l}
   armBackR {angle=50, bDrawDebug=false, endJoint=lowerarm_r, poseJoint=pose_arm_bk_r, startJoint=upperarm_r}
   armForwardDownL {angle=40, bDrawDebug=false, endJoint=lowerarm_l, poseJoint=pose_arm_fwd_dn_l, startJoint=upperarm_l}
   armForwardDownR {angle=40, bDrawDebug=false, endJoint=lowerarm_r, poseJoint=pose_arm_fwd_dn_r, startJoint=upperarm_r}
   armForwardL {angle=50, bDrawDebug=false, endJoint=lowerarm_l, poseJoint=pose_arm_fwd_l, startJoint=upperarm_l}
   armForwardR {angle=50, bDrawDebug=false, endJoint=lowerarm_r, poseJoint=pose_arm_fwd_r, startJoint=upperarm_r}
   armForwardUpL {angle=45, bDrawDebug=false, endJoint=lowerarm_l, poseJoint=pose_arm_fwd_up_l, startJoint=upperarm_l}
   armForwardUpR {angle=45, bDrawDebug=false, endJoint=lowerarm_r, poseJoint=pose_arm_fwd_up_r, startJoint=upperarm_r}
   armUpL {angle=70, bDrawDebug=false, endJoint=lowerarm_l, poseJoint=pose_arm_up_l, startJoint=upperarm_l}
   armUpR {angle=70, bDrawDebug=false, endJoint=lowerarm_r, poseJoint=pose_arm_up_r, startJoint=upperarm_r}
  }
 }
 thigh_rot_l {type=number}
 thigh_rot_r {type=number}
]
